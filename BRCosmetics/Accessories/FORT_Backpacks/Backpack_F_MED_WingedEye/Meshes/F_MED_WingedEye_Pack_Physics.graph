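# Physics asset for the Winged Eye backpack cosmetic, giving its skeleton bodies and joints for secondary motion of the wings and tail.
# A sphere covers the central body, with small capsules on two base-and-wing pairs and a based four-segment tail chain.
# Base joints to the body have free swing and twist, while tail segments carry tight angular limits and wings allow much wider swing with little twist.
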PhysicsAsset /BRCosmetics/Accessories/FORT_Backpacks/Backpack_F_MED_WingedEye/Meshes/F_MED_WingedEye_Pack_Physics
bodies {
 dyn_body [{shapeType=Sphere, radius=9, shapeName=dyn_body_sphere, center=<-0.5257 0 0>}]
 dyn_tail_01 [{shapeType=Capsule, length=2, radius=0.529, shapeName=dyn_tail_01_capsule, center=<0.6 0 0>, rotation=<rot 0 0 36.52>}]
 dyn_tail_02 [{shapeType=Capsule, length=1.2912, radius=0.5602, shapeName=dyn_tail_02_capsule, center=<-0.5 0 0>}]
 dyn_tail_03 [{shapeType=Capsule, length=1.2512, radius=0.7454, shapeName=dyn_tail_03_capsule, center=<-0.1 0 0>}]
 dyn_tail_04 [{shapeType=Capsule, length=1, radius=0.5, shapeName=dyn_tail_04_capsule}]
 dyn_tail_basa [{shapeType=Capsule, length=0.505, radius=0.5, shapeName=dyn_tail_basa_capsule, rotation=<rot 0 -90 0>}]
 dyn_wing_basa_l [{shapeType=Capsule, length=0.505, radius=0.505, shapeName=dyn_wing_basa_l_capsule, rotation=<rot 0 -90 0>}]
 dyn_wing_basa_r [{shapeType=Capsule, length=0.505, radius=0.505, shapeName=dyn_wing_basa_r_capsule, rotation=<rot 0 -90 0>}]
 dyn_wing_l [{shapeType=Capsule, length=2, radius=1, shapeName=dyn_wing_l_capsule, center=<0.5 0 0>, rotation=<rot 90 0 0>}]
 dyn_wing_r [{shapeType=Capsule, length=2, radius=1, shapeName=dyn_wing_r_capsule, center=<0.5 0 0>, rotation=<rot 90 90 90>}]
}
constraints [
 {bone1Name=dyn_tail_01, bone2Name=dyn_tail_basa, swing1LimitDegrees=1, swing2LimitDegrees=15, twistLimitDegrees=6}
 {bone1Name=dyn_tail_02, bone2Name=dyn_tail_01, swing1LimitDegrees=4, swing2LimitDegrees=10, twistLimitDegrees=3}
 {bone1Name=dyn_tail_03, bone2Name=dyn_tail_02, swing1LimitDegrees=4, swing2LimitDegrees=8, twistLimitDegrees=3}
 {bone1Name=dyn_tail_04, bone2Name=dyn_tail_03, swing1LimitDegrees=10, swing2LimitDegrees=20, twistLimitDegrees=3}
 {bone1Name=dyn_wing_l, bone2Name=dyn_wing_basa_l, swing1LimitDegrees=35, swing2LimitDegrees=100, twistLimitDegrees=5}
 {bone1Name=dyn_wing_r, bone2Name=dyn_wing_basa_r, swing1LimitDegrees=35, swing2LimitDegrees=100, twistLimitDegrees=5}
 {
  bone1Name=dyn_wing_basa_l
  bone2Name=dyn_body
  swing1LimitDegrees=45
  swing1Motion=Free
  swing2LimitDegrees=45
  swing2Motion=Free
  twistLimitDegrees=45
  twistMotion=Free
 }
 {
  bone1Name=dyn_wing_basa_r
  bone2Name=dyn_body
  swing1LimitDegrees=45
  swing1Motion=Free
  swing2LimitDegrees=45
  swing2Motion=Free
  twistLimitDegrees=45
  twistMotion=Free
 }
 {
  bone1Name=dyn_tail_basa
  bone2Name=dyn_body
  swing1LimitDegrees=45
  swing1Motion=Free
  swing2LimitDegrees=45
  swing2Motion=Free
  twistLimitDegrees=45
  twistMotion=Free
 }
]
